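; B_IcyPatch is an actor managing a group of timed ice patches that apply the IcyShins effect and slide boosts to pawns.
; The server adds points directly or through queued EQS queries, refreshes nearby lifespans, and sends updates to clients.
; A box refitted around the patches starts timed per-pawn distance checks on the server.
; Clients spawn ground-aligned decals and share particles through the game-state patch manager component.
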(event EventBeginPlay
  (bind _returnvalue (Game|GetGameState))
  ;[0.01367 0.15 0.02516 0.5] Begin Play
  (|CallParentFunction)
  (Utilities|IsValid _returnvalue
    (:Is%20Valid
      (Variables|Default|SetPatchManagerComponent (Actor|GetComponentbyClass _returnvalue "/TeaCakeIcyShins/Gameplay/B_FortGameStateComponent_IcyPatchManager.B_FortGameStateComponent_IcyPatchManager_C"))
      (if (> (Variables|Debug|GetDebugMode) 0)
        (bind _returnvalue_2 (Utilities|Time|SetTimerbyEvent (EventDispatchers|CreateEvent :!event DebugDrawPatches) 0.1 true))
        (Variables|Debug|SetDebugDrawUpdateHandle _returnvalue_2)))
    (:Is%20Not%20Valid
      (Fortnite|FortShippingLog self "BeingPlay: Patch Manager Component not found" true))))

(event Custom|GeneratePatchAtLocation
  (params (Location :type Vector))
  ;[0.001136 0.15 0.06962 0.5] Patch Generation
  ;;[0.05459 0.15 0.03517 0.5] Generate Patch at Location
  (Networking|SwitchHasAuthority
    (:Authority
      (PatchManagement|AddNewPatchBatch (Utilities|Array|MakeArray Location))
      (CallFunction|UpdateCleanupTimer (Variables|Settings|Patch|GetPatchLifeSpan)))))

(event Custom|GeneratePatchesAtLocation
  (params (BatchLocation :type Vector) (EQS%20Template :type Env%20Query%20Object%20Reference))
  ;[0.001136 0.15 0.06962 0.5] Patch Generation
  ;;[0.05459 0.15 0.03517 0.5] Generate Patch at Location
  (Networking|SwitchHasAuthority
    (:Authority
      (EQS|RequestEQSQuery EQS%20Template BatchLocation))))

(event Custom|PatchPlacementEQSFinished
  (params (QueryInstance :type EQS%20Query%20Instance%20Object%20Reference) (QueryStatus :type EEnvQueryStatus%20Enum :default "Processing"))
  ;[0.001136 0.15 0.06962 0.5] Patch Generation
  ;;[0.02724 0.15 0.09501 0.5] EQS Finished
  (Networking|SwitchHasAuthority
    (:Authority
      (Variables|Runtime|EQS|SetEQSQueryInstance 0)
      (Utilities|FlowControl|Sequence
        (:then_0
          (switch Utilities|FlowControl|Switch|SwitchonEEnvQueryStatus QueryStatus
            (:Success
              (Utilities|IsValid QueryInstance
                (:Is%20Valid
                  (bind _resultlocations (AI|EQS|GetQueryResultsasLocations QueryInstance))
                  (PatchManagement|AddNewPatchBatch _resultlocations)
                  (CallFunction|UpdateCleanupTimer (Variables|Settings|Patch|GetPatchLifeSpan)))))))
        (:then_1
          (EQS|RunNextEQSinQueue))))))

(event Custom|Client_UpdatePatchLocation
  (params (AddedPatchLocations :type Array%20of%20Icy%20Patch%20Data%20Structures) (UpdatedLocationAverage :type Vector%20%28by%20ref%29))
  ;[0.001136 0.15 0.06962 0.5] Patch Generation
  ;;[0.1149 0.04815 0.15 0.5] Replicate patches
  (Networking|CanExecuteCosmeticEvents
    (:True
      (PatchManagement|CleanupDeadPatches)
      (PatchManagement|RefreshPatches UpdatedLocationAverage) ; Refresh location is replicated to handle \nthe case where batch is added on top of \nexsisting patches, and no new patches are added
      (PatchManagement|AddClientPatches AddedPatchLocations)
      (CallFunction|UpdateCleanupTimer (Variables|Settings|Patch|GetPatchLifeSpan)) ; For first patch when timer \nhasn't been started yet
    )))

(event Custom|UpdateCleanupTimer
  (params (Delay :type Float%20%28single-precision%29) (ForceRefresh :type Boolean))
  ;[0.001136 0.15 0.06962 0.5] Patch Generation
  ;;[0.15 0.03198 0.0444 0.5] Start cleanup timer if not active
  (if (or ForceRefresh (not (Utilities|Time|IsTimerActivebyHandle (Variables|Runtime|Patch|GetCleanupTimerHandle))))
    (bind _returnvalue (Utilities|Time|SetTimerbyEvent (EventDispatchers|CreateEvent :!event CleanupDeadPatches) (Math|Float|Max(Float) Delay 0.1)))
    (Variables|Runtime|Patch|SetCleanupTimerHandle _returnvalue)))

(event OnComponentBeginOverlap(Box)
  (params (OverlappedComponent :type Primitive%20Component%20Object%20Reference) (OtherActor :type Actor%20Object%20Reference) (OtherComp :type Primitive%20Component%20Object%20Reference) (OtherBodyIndex :type Integer) (bFromSweep :type Boolean) (SweepResult :type Hit%20Result%20Structure%20%28by%20ref%29))
  ;[0.03392 0.03947 0.15 0.5] Player interaction
  ;;[0.06679 0.15 0.04468 0.5] Pawn Start Overlap
  (Networking|SwitchHasAuthority
    (:Authority
      (bind _asfort_pawn (Utilities|Casting|CastToFortPawn OtherActor))
      (bind _returnvalue (Utilities|Array|AddUnique (Variables|Runtime|PawnInteraction|GetOverlappingPawns) _asfort_pawn))
      (CallFunction|UpdateOverlappingPawns)
      (if (and (>= _returnvalue 0) (not (Utilities|Time|IsTimerActivebyHandle (Variables|Runtime|PawnInteraction|GetPawnOverlapUpdateHandle))))
        (bind _returnvalue_2 (Utilities|Time|SetTimerbyEvent (EventDispatchers|CreateEvent :!event Custom|UpdateOverlappingPawns) (Variables|Settings|PawnInteraction|GetPawnOverlapUpdateRate) true))
        (Variables|Runtime|PawnInteraction|SetPawnOverlapUpdateHandle _returnvalue_2)))))

(event Custom|UpdateOverlappingPawns
  ;[0.03392 0.03947 0.15 0.5] Player interaction
  ;;[0.03605 0.15 0.1088 0.5] Update Overlapping Pawns
  (Networking|SwitchHasAuthority
    (:Authority
      (for _array_element (Variables|Runtime|PawnInteraction|GetOverlappingPawns)
        (Utilities|IsValid _array_element
          (:Is%20Valid
            (if (.bOverlappingPatch (PatchHelpers|GetNearbyPatches (Transformation|GetActorLocation _array_element) (Variables|Settings|PawnInteraction|GetMinPlayerDistanceSquared) true))
              (CallFunction|ApplyPatchEffectstoPawn _array_element))))))))

(event Custom|ApplyPatchEffectsToPawn
  (params (PlayerPawn :type Fort%20Pawn%20Object%20Reference))
  ;[0.03392 0.03947 0.15 0.5] Player interaction
  ;;[0.04139 0.1459 0.15 0.5] Apply Patch Effects to Pawn
  (Networking|SwitchHasAuthority
    (:Authority
      (Utilities|FlowControl|Sequence
        (:then_0
          (if (Math|Boolean|ANDBoolean (GameplayTags|HasMatchingGameplayTag PlayerPawn "(TagName=\"MovementMode.Sliding\")") (not (GameplayTags|HasMatchingGameplayTag PlayerPawn "(TagName=\"Gameplay.Status.IcyShins.Applied\")")) (ScalableFloat|GetValueasBool (Variables|Default|GetEnableBoostWhileAlreadySliding)))
            (CallFunction|ApplySlideBoostMME PlayerPawn)))
        (:then_1
          (bind _returnvalue (Spec|MakeSpecHandle(ByClass) (Variables|Settings|PawnInteraction|GetGE_IcyShins) (Variables|Actor|GetInstigator) self))
          (Environment|ApplyGlobalEnvironmentGameplayEffectSpectoActor PlayerPawn _returnvalue))))))

(event Custom|ApplySlideBoostMME
  (params (TargetPawn :type Fort%20Pawn%20Object%20Reference))
  ;[0.03392 0.03947 0.15 0.5] Player interaction
  ;; Temp. Probably want to make a way to activate on server
  (Utilities|IsValid TargetPawn
    (:Is%20Valid
      (Class|FortMovementModeExtBlueprintLibrary|ActivateRegisteredMovementModeExtension (Variables|Default|GetSlideBoostMME) TargetPawn))))

(event OnComponentEndOverlap(Box)
  (params (OverlappedComponent :type Primitive%20Component%20Object%20Reference) (OtherActor :type Actor%20Object%20Reference) (OtherComp :type Primitive%20Component%20Object%20Reference) (OtherBodyIndex :type Integer))
  (bind _overlappingpawns (Variables|Runtime|PawnInteraction|GetOverlappingPawns))
  ;[0.03392 0.03947 0.15 0.5] Player interaction
  ;;[0.15 0.07495 0.02073 0.5] Pawn End Overlap
  (Networking|SwitchHasAuthority
    (:Authority
      (bind _asfort_pawn (Utilities|Casting|CastToFortPawn OtherActor))
      (bind _returnvalue (Utilities|Array|RemoveItem _overlappingpawns _asfort_pawn))
      (if (and _returnvalue (Utilities|Array|IsEmpty _overlappingpawns))
        (Utilities|Time|ClearandInvalidateTimerbyHandle (Variables|Runtime|PawnInteraction|GetPawnOverlapUpdateHandle))))))

(event EventEndPlay
  (params (EndPlayReason :type EEndPlayReason%20Enum :default "Destroyed"))
  ;[0.15 0.04183 0.03896 0.5] End Play
  (|CallParentFunction EndPlayReason)
  (Utilities|FlowControl|Sequence
    (:then_0
      (bind _eqsqueryinstance (Variables|Runtime|EQS|GetEQSQueryInstance)
        (:then
          (AI|EQS|UnbindEventfromOnQueryFinished _eqsqueryinstance (EventDispatchers|CreateEvent :!event Custom|PatchPlacementEQSFinished)))
        (:else
          _))
      (Utilities|Time|ClearandInvalidateTimerbyHandle (Variables|Runtime|PawnInteraction|GetPawnOverlapUpdateHandle))
      (Utilities|Time|ClearandInvalidateTimerbyHandle (Variables|Runtime|Patch|GetCleanupTimerHandle)))
    (:then_1
      ;;[0.15 0.03653 0.08229 0.5] Remove patches from particles to allow particles to get cleaned up
      (Networking|CanExecuteCosmeticEvents
        (:True
          (for _array_element (Variables|Runtime|Patch|GetPatches)
            (bind _patchmanagercomponent (Variables|Default|GetPatchManagerComponent)
              (:then
                (Default|ModifyParticlePatchCount _patchmanagercomponent (.ParticleIndex_8_DCFDBC164B67F81DDED7538FEB3AB70D (Utilities|Struct|BreakIcyPatchData _array_element)) -1))
              (:else))))))))

(fn AddClientPatches
  (params (NewPatches :type Array%20of%20Icy%20Patch%20Data%20Structures))
  (bind _requestedparticles (LocalVariables|GetRequestedParticles))
  (bind _patches (Variables|Runtime|Patch|GetPatches))
  ;[0.1191 0.03903 0.15 0.5] Add patches and create new particle if needed
  ;;[0.04013 0.15 0.05644 0.5] Add patches
  (for _array_element NewPatches
    (bind (_newpatchindex _bneedsparticle _suggestedparticle) (PatchManagement|AddPatch _array_element))
    (if _bneedsparticle
      (Utilities|Array|Add (LocalVariables|GetPatchesNeedingParticles) _newpatchindex)
      (Utilities|Array|Add (LocalVariables|GetRequestedParticles) _suggestedparticle)))
  ;;[0.15 0.1137 0.04552 0.5] If new particle is needed, calculate an average for a new particle that can be used by all patches that couldn't use an exsiting particle
  ;; Assumes patches being added at the same time are near eachother
  (for _array_element_2 _requestedparticles
    (LocalVariables|SetCumulativeLocation (+ (Utilities|Struct|BreakIcyPatchParticleData _array_element_2) (LocalVariables|GetCumulativeLocation)))
    (LocalVariables|SetCumulativeNormal (+ (.Normal_5_C845A950432BCEE1A7CFB3BAB26E78F0 (Utilities|Struct|BreakIcyPatchParticleData _array_element_2)) (LocalVariables|GetCumulativeNormal))))
  ;;[0.07162 0.15 0.04842 0.5] Construct the new average particle, add it to the patch manager, and update patches that use it
  (bind _output_get (LocalVariables|SetNumRequestedParticles (Utilities|Array|Length _requestedparticles)))
  (bind _output_get_2 (LocalVariables|SetNewAveragedParticle (Utilities|Struct|MakeIcyPatchParticleData (/ (LocalVariables|GetCumulativeLocation) (Math|Conversions|ToFloat(Integer) _output_get)) (/ (LocalVariables|GetCumulativeNormal) (Math|Conversions|ToFloat(Integer) _output_get)) _output_get)))
  (bind _patchmanagercomponent (Variables|Default|GetPatchManagerComponent)
    (:then
      (bind _index (Default|AddParticle _patchmanagercomponent _output_get_2))
      (for _array_element_3 (LocalVariables|GetPatchesNeedingParticles)
        (if (Utilities|Array|IsValidIndex _patches _array_element_3)
          (Utilities|Array|SetArrayElem _patches _array_element_3 (DataHelpers|SetPatchParticle (Utilities|Array|Get(aref) _patches _array_element_3) _index)))))
    (:else)))

(fn AddNewPatchBatch
  (params (PatchLocations :type Array%20of%20Vectors))
  (bind _patchlocations (Variables|GetPatchLocations))
  (bind _returnvalue (Utilities|Array|Length _patchlocations))
  (bind _returnvalue_2 (Math|Conversions|ToFloat(Integer) _returnvalue))
  (bind _returnvalue_3 (Fortnite|Sync|Time|GetServerWorldTimeSeconds))
  (bind _cumulativepatchlocations (LocalVariables|GetCumulativePatchLocations))
  (bind _returnvalue_4 (/ _cumulativepatchlocations _returnvalue_2))
  ;[0.02392 0.15 0.06182 0.5] Add new batch of locations to tracked patches. Potentially limit to a certain amount per batch
  (Networking|SwitchHasAuthority
    (:Authority
      (PatchManagement|CleanupDeadPatches) ; Remove any dead patches before \nadding new ones
      (for _array_element PatchLocations
        (LocalVariables|SetCumulativePatchLocations (+ _array_element (LocalVariables|GetCumulativePatchLocations))) ; Used to calculate average batch location\nCaclculated even for patches that aren't added to \nrefresh exsiting patches in that location
        (if (not (.bOverlappingPatch (PatchHelpers|GetNearbyPatches _array_element (Variables|Settings|Patch|GetMinPatchSpacingDistanceSquared) true)))
          (Utilities|Array|Add (Variables|Runtime|Patch|GetPatches) (Utilities|Struct|MakeIcyPatchData _array_element _returnvalue_3))
          (Utilities|Array|Add (LocalVariables|GetPatchesAdded) (Utilities|Struct|MakeIcyPatchData _array_element _returnvalue_3))))
      ;[0.1174 0.15 0.05899 0.5] Recalculate patch area, update the lifespan of all nearby patches, and send new patches to client
      (if (not (Utilities|Array|IsEmpty _patchlocations))
        (PatchManagement|CalculateNewPatchArea)
        (PatchManagement|RefreshPatches _returnvalue_4)
        (CallFunction|ClientUpdatePatchLocation (LocalVariables|GetPatchesAdded) _returnvalue_4)))))

(fn AddPatch
  (params (Patch :type Icy%20Patch%20Data%20Structure))
  (bind _location_2_4a09594d44e582f062dc1cabe0e0b05d (Utilities|Struct|BreakIcyPatchData Patch))
  (bind (_location_4_0b01d8c84ddd74cd5df78d9cbe87e092 _normal_5_c845a950432bcee1a7cfb3bab26e78f0 _patchcount_17_00e3f4e549ca99fc81c81586268f2443) (Utilities|Struct|BreakIcyPatchParticleData (LocalVariables|GetNewParticle)))
  (bind (_bblockinghit _binitialoverlap _time _distance _location _impactpoint _normal) (Collision|BreakHitResult (LocalVariables|GetGroundHit)))
  (bind (_bblockinghit_2 _binitialoverlap_2 _time_2 _distance_2 _location_2 _impactpoint_2 _normal_2) (Collision|BreakHitResult (LocalVariables|GetGroundHit)))
  (bind _patchlifespan (Variables|Settings|Patch|GetPatchLifeSpan))
  (bind (_patchtoadd_location_2_4a09594d44e582f062dc1cabe0e0b05d _patchtoadd_spawntime_5_00ab2d44400cb05fdc7265903521a906 _patchtoadd_particleindex_8_dcfdbc164b67f81dded7538feb3ab70d) (Utilities|Struct|BreakIcyPatchData (Variables|GetPatchToAdd_Location_2_4A09594D44E582F062DC1CABE0E0B05D)))
  (bind _patches (Variables|Runtime|Patch|GetPatches))
  (LocalVariables|SetPatchToAdd Patch)
  (bind _outhit (Collision|LineTraceByChannel _location_2_4a09594d44e582f062dc1cabe0e0b05d (Utilities|Operators|Add _location_2_4a09594d44e582f062dc1cabe0e0b05d (Math|Vector|MakeVector 0.0 0.0 (Variables|Settings|Patch|GetDecalOrientationTraceLength))) "TraceTypeQuery10")) ; Unfortunetly I don't think EQS gives a normal, which is needed to place the patch correctly\nThis small trace is just used to figure out the ground normal
  (LocalVariables|SetGroundHit _outhit)
  (Utilities|FlowControl|Sequence
    (:then_0
      ;[0.1083 0.04751 0.15 0.5] Add decal component flush against ground at patch location
      (bind _returnvalue (Rendering|Decal|SpawnDecalatLocation (Variables|Settings|Decal|GetDecal%20Material) (Variables|Settings|Decal|GetDecalSize) (Math|Vector|SelectVector _location _location_2_4a09594d44e582f062dc1cabe0e0b05d _bblockinghit) (Math|Rotator|SelectRotator (Math|Rotator|MakeRotfromX (* _normal -1.0)) "-90.000000,0.000000,0.000000" _bblockinghit) _patchlifespan))
      (Rendering|Components|Decal|SetFadeOut _returnvalue _patchlifespan (Variables|Settings|Decal|GetDecalFadeOutTime) false)
      (LocalVariables|SetPatchToAdd (DataHelpers|SetPatchDecal (LocalVariables|GetPatchToAdd) _returnvalue)))
    (:then_1
      ;[0.2917 0.06838 0.262 0.5] Add patch to existing particle, or request new particle
      (bind _output_get (LocalVariables|SetNearbyPatches (PatchHelpers|GetNearbyPatches _location_2_4a09594d44e582f062dc1cabe0e0b05d)))
      ;;[0.03053 0.15 0.03946 0.5] For each nearby patch, see if this new patch can use the same particle the nearby patch is using
      (bind _array_element (Utilities|Array|ForEachLoopwithBreak _output_get)
        (:LoopBody
          (if (Utilities|Array|IsValidIndex _patches _array_element)
            (if (not (Utilities|Array|ContainsItem (LocalVariables|GetAttemptedParticleIndicies) (.ParticleIndex_8_DCFDBC164B67F81DDED7538FEB3AB70D (Utilities|Struct|BreakIcyPatchData (Utilities|Array|Get(aref) _patches _array_element)))))
              (bind _patchmanagercomponent (Variables|Default|GetPatchManagerComponent)
                (:then
                  (if (Default|CanPatchUseParticle _patchmanagercomponent (.ParticleIndex_8_DCFDBC164B67F81DDED7538FEB3AB70D (Utilities|Struct|BreakIcyPatchData (Utilities|Array|Get(aref) _patches _array_element))) (Variables|GetPatch) (LocalVariables|GetGroundHit))
                    (Default|ModifyParticlePatchCount _patchmanagercomponent (.ParticleIndex_8_DCFDBC164B67F81DDED7538FEB3AB70D (Utilities|Struct|BreakIcyPatchData (Utilities|Array|Get(aref) _patches _array_element))) 1)
                    (LocalVariables|SetPatchToAdd (DataHelpers|SetPatchParticle (LocalVariables|GetPatchToAdd) (.ParticleIndex_8_DCFDBC164B67F81DDED7538FEB3AB70D (Utilities|Struct|BreakIcyPatchData (Utilities|Array|Get(aref) _patches _array_element))))))
                  (Utilities|Array|Add (LocalVariables|GetAttemptedParticleIndicies) (.ParticleIndex_8_DCFDBC164B67F81DDED7538FEB3AB70D (Utilities|Struct|BreakIcyPatchData (Utilities|Array|Get(aref) _patches _array_element)))))
                (:else)))))
        (:Completed
          _))
      ;;[0.01392 0.06052 0.15 0.5] If particle couldn't be shared, request a new particle
      (if (< _patchtoadd_particleindex_8_dcfdbc164b67f81dded7538feb3ab70d 0)
        (LocalVariables|SetNewParticle (Utilities|Struct|MakeIcyPatchParticleData _patchtoadd_location_2_4a09594d44e582f062dc1cabe0e0b05d _normal_2 1))))
    (:then_2
      ;[0.15 0.05 0.06004 0.5] Return the new patch, and whether it needs a new particle
      (bind _returnvalue_2 (Utilities|Array|Add (Variables|Runtime|Patch|GetPatches) (LocalVariables|GetPatchToAdd)))
      (return _returnvalue_2 (> _patchcount_17_00e3f4e549ca99fc81c81586268f2443 0) (LocalVariables|GetNewParticle)))))

(fn CalculateNewPatchArea
  (bind (_patchcenter_x _patchcenter_y _patchcenter_z) (Math|Vector|BreakVector (Variables|GetPatchCenter_X)))
  (bind (_newboxscale_x _newboxscale_y _newboxscale_z) (Math|Vector|BreakVector (Variables|GetNewBoxScale_X)))
  (bind _box (Variables|Default|GetBox))
  (Collision|SetCollisionEnabled (Variables|Default|GetBox) "NoCollision") ; Disable collision while we update component to \nprevent multiple overlap checks
  (Utilities|FlowControl|Sequence
    (:then_0
      ;[0.15 0.1186 0.03862 0.5] Calculate the new center of the patch by getting the average of all locations
      ; Note: It is expected that any dead patches would have been removed by the time this function is called
      (for _array_element (Variables|Runtime|Patch|GetPatches)
        (LocalVariables|SetAdditivePatchLocations (+ (Utilities|Struct|BreakIcyPatchData _array_element) (LocalVariables|GetAdditivePatchLocations))))
      (Variables|Runtime|Patch|SetPatchCenter (/ (LocalVariables|GetAdditivePatchLocations) (Utilities|Array|Length (Variables|Runtime|Patch|GetPatches)))))
    (:then_1
      ;[0.15 0.09972 0.06225 0.5] Calculate the patch locations that are furthest from the center on horizontally and vertically (X and Z box extents)
      (for _array_element_2 (Variables|Runtime|Patch|GetPatches)
        (bind _output_get (LocalVariables|SetCurrentPatchDistance (Math|Vector|DistanceSquared(Vector) (Utilities|Struct|BreakIcyPatchData _array_element_2) (Variables|Runtime|Patch|GetPatchCenter)))) ; Prevent calculating distance multiple times\n by caching value
        (Utilities|FlowControl|Sequence
          (:then_0
            (if (> _output_get (LocalVariables|GetFurthestPatchDistance))
              (LocalVariables|SetFurthestPatchDistance (LocalVariables|GetCurrentPatchDistance))
              (LocalVariables|SetFurthestPatchLocation (Utilities|Struct|BreakIcyPatchData _array_element_2))))
          (:then_1
            (if (> (Math|Float|Absolute(Float) (- (.z (Utilities|Struct|BreakIcyPatchData _array_element_2)) _patchcenter_z)) (LocalVariables|GetFurthestHeightPatch))
              (LocalVariables|SetFurthestHeightPatch (Math|Float|Absolute(Float) (- (.z (Utilities|Struct|BreakIcyPatchData _array_element_2)) _patchcenter_z)))))))
      (Variables|Runtime|Patch|SetPatchRadius (Math|Float|Sqrt (LocalVariables|GetFurthestPatchDistance))) ; Update the patch radius used for determining \nif new patch location should be added to this actor
      (Transformation|SetWorldLocationAndRotation (Variables|Default|GetBox) (Variables|Runtime|Patch|GetPatchCenter) (Math|Conversions|RotationFromXVector (* (- (LocalVariables|GetFurthestPatchLocation) (Variables|Runtime|Patch|GetPatchCenter)) "1.000000,1.000000,0.000000")) false true) ; Now that we know the center and furthest point, \nwe can update the location and make rotation with \nX axis pointing towards furthest point
      (Variables|SetNewBoxScale_X (Math|Vector|MakeVector (Variables|Runtime|Patch|GetPatchRadius) 0.0 (LocalVariables|GetFurthestHeightPatch))) ; Save the calculated X and Z extents\nWe'll find the Y next
    )
    (:then_2
      ;[0.15 0.0637 0.04642 0.5] Calculate the patch location that's furthest perpendicular of the direction to the actual furthest (Y box extents)
      (LocalVariables|SetFurthestPatchDistance 0.0) ; Reuse cached futhest variable
      (LocalVariables|SetBoxYAxis (Transformation|GetRightVector (Variables|Default|GetBox)))
      (for _array_element_3 (Variables|Runtime|Patch|GetPatches)
        (bind _output_get_2 (LocalVariables|SetCurrentPatchDistance (Math|Vector|DotProduct (LocalVariables|GetBoxYAxis) (- (Utilities|Struct|BreakIcyPatchData _array_element_3) (Variables|Runtime|Patch|GetPatchCenter))))) ; Prevent calculating distance multiple times\n by caching value
        (if (> _output_get_2 (LocalVariables|GetFurthestPatchDistance))
          (LocalVariables|SetFurthestPatchDistance (LocalVariables|GetCurrentPatchDistance))))
      (bind _output_get_3 (Variables|SetNewBoxScale_X (Math|Vector|MakeVector _newboxscale_x (LocalVariables|GetFurthestPatchDistance) _newboxscale_z)))
      (LocalVariables|SetNewBoxScale (+ _output_get_3 (Variables|Settings|Patch|GetAdditionalOverlapSize))) ; Add additional overlap buffer
    )
    (:then_3
      ;[0.04375 0.15 0.1269 0.5] Update overlap component
      (Components|Box|SetBoxExtent _box (LocalVariables|GetNewBoxScale) false)
      (Collision|SetCollisionEnabled _box "QueryOnly") ; Re-enable collision now that we know the final size
    )))

(fn CleanupDeadPatches
  (bind _patches (Variables|Runtime|Patch|GetPatches))
  (if (not (Utilities|Array|IsEmpty _patches))
    (bind (_arrayindex _arrayelement) (Utilities|Array|ReverseforEachLoop _patches)
      (:LoopBody
        ;[0.15 0.04333 0.04906 0.5] Remove any patches that have been alive longer than the PatchLifeSpan. Update the particle so it can be removed if this is the last patch using that particle
        (if (> (- (Fortnite|Sync|Time|GetServerWorldTimeSeconds) (.SpawnTime_5_00AB2D44400CB05FDC7265903521A906 (Utilities|Struct|BreakIcyPatchData _arrayelement))) (Variables|Settings|Patch|GetPatchLifeSpan))
          (Utilities|FlowControl|Sequence
            (:then_0
              (bind _patchmanagercomponent (Variables|Default|GetPatchManagerComponent)
                (:then
                  (Default|ModifyParticlePatchCount _patchmanagercomponent (.ParticleIndex_8_DCFDBC164B67F81DDED7538FEB3AB70D (Utilities|Struct|BreakIcyPatchData _arrayelement)) -1))
                (:else)))
            (:then_1
              (Utilities|Array|RemoveIndex (Variables|Runtime|Patch|GetPatches) _arrayindex)
              (Math|Integer|IncrementInt (LocalVariables|GetPatchesRemoved))))
          (else
            ;[0.15 0.1371 0.07011 0.5] Keep track of oldest alive patch for setting next cleanup timer
            (if (or (< (.SpawnTime_5_00AB2D44400CB05FDC7265903521A906 (Utilities|Struct|BreakIcyPatchData _arrayelement)) (LocalVariables|GetMinSpawnTime)) (< (LocalVariables|GetMinSpawnTime) 0))
              (LocalVariables|SetMinSpawnTime (.SpawnTime_5_00AB2D44400CB05FDC7265903521A906 (Utilities|Struct|BreakIcyPatchData _arrayelement)))))))
      (:Completed
        ;[0.15 0.06246 0.0171 0.5] After cleanup
        (if (Utilities|Array|IsEmpty (Variables|Runtime|Patch|GetPatches))
          ;;[0.15 0.01011 0.004376 0.5] Destroy actor if there are no more patches
          (Networking|SwitchHasAuthority
            (:Authority
              (Actor|SetLifeSpan (Variables|Settings|GetLifespanAfterLastPatch))))
          (else
            ;;[0.04038 0.15 0.1195 0.5] Otherwise, handle patches being removed
            (CallFunction|UpdateCleanupTimer (- (Variables|Settings|Patch|GetPatchLifeSpan) (- (Fortnite|Sync|Time|GetServerWorldTimeSeconds) (LocalVariables|GetMinSpawnTime))) true) ; Need to force refresh because timer \nis technically still active after calling this function
            (if (> (LocalVariables|GetPatchesRemoved) 0)
              (PatchManagement|CalculateNewPatchArea))))))))

(fn DebugDrawPatches
  (bind _returnvalue (Networking|IsDedicatedServer))
  (bind (_decalsize_x _decalsize_y) (Math|Vector|BreakVector (Variables|GetDecalSize_X)))
  (switch int (Variables|Debug|GetDebugMode)
    (:1
      (if _returnvalue
        _))
    (:2
      (if (not _returnvalue)
        _))
    (:3
      _))
  (Rendering|Debug|DrawDebugBox (Transformation|GetWorldLocation (Variables|Default|GetBox)) (Components|Box|GetScaledBoxExtent (Variables|Default|GetBox)) "(R=0.000000,G=0.948661,B=1.000000,A=1.000000)" (Transformation|GetWorldRotation (Variables|Default|GetBox)) (Variables|Debug|GetDebugDrawDuration))
  (for _array_element (Variables|Runtime|Patch|GetPatches)
    (Rendering|Debug|DrawDebugSphere (Utilities|Struct|BreakIcyPatchData _array_element) (* _decalsize_y (Math|Float|SelectFloat 0.5 0.25 (Networking|IsDedicatedServer))) 12 (Math|Color|MakeColor (/ (- (Fortnite|Sync|Time|GetServerWorldTimeSeconds) (.SpawnTime_5_00AB2D44400CB05FDC7265903521A906 (Utilities|Struct|BreakIcyPatchData _array_element))) (Variables|Settings|Patch|GetPatchLifeSpan)) (- 1.0 (/ (- (Fortnite|Sync|Time|GetServerWorldTimeSeconds) (.SpawnTime_5_00AB2D44400CB05FDC7265903521A906 (Utilities|Struct|BreakIcyPatchData _array_element))) (Variables|Settings|Patch|GetPatchLifeSpan))) (Math|Float|SelectFloat 0.0 1.0 (Networking|IsDedicatedServer))) (Variables|Debug|GetDebugDrawDuration))))

(fn GetNearbyPatches
  (params (Location :type Vector) (OverlapRadiusSquared :type Float%20%28double-precision%29) (bStopAtFirstOverlap :type Boolean))
  (bind _overlapradiussquared (Variables|GetOverlapRadiusSquared))
  ;[0.15 0.1322 0.05433 0.5] Helper function to get all patches within a nearby radius and/or check if a location is overlapping an existing patch
  (for (_array_element _array_index) (Variables|Runtime|Patch|GetPatches)
    (bind _output_get (LocalVariables|SetCurrentPatchDistanceSquared (Math|Vector|Distance2DSquared(Vector) (Utilities|Struct|BreakIcyPatchData _array_element) Location)))
    (Utilities|FlowControl|Sequence
      (:then_0
        ;;[0.04725 0.15 0.0171 0.5] Find patches within nearby radius
        (if (< _output_get (Variables|Settings|Patch|GetNearbyRadiusSquared))
          (Utilities|Array|Add (LocalVariables|GetPatchesIndiciesNearby) _array_index)))
      (:then_1
        ;;[0.15 0.0525 0.0525 0.5] Check if exsisting patch is overlapping location
        (if (< _output_get _overlapradiussquared)
          (LocalVariables|SetbOverlapping (< _output_get _overlapradiussquared))
          (if (Variables|GetbStopAtFirstOverlap)
            (return (LocalVariables|GetPatchesIndiciesNearby) (LocalVariables|GetbOverlapping)))))))
  (return (LocalVariables|GetPatchesIndiciesNearby) (LocalVariables|GetbOverlapping)))

(fn IsPatchLocationValid
  (params (PatchLocation :type Vector))
  (for _array_element (Variables|Runtime|Patch|GetPatches)
    (if (< (Math|Vector|Distance2DSquared(Vector) (Utilities|Struct|BreakIcyPatchData _array_element) PatchLocation) (Variables|Settings|Patch|GetMinPatchSpacingDistanceSquared))
      (return false)))
  (return true))

(fn RefreshPatches
  (params (RefreshLocation :type Vector))
  (bind _patches (Variables|Runtime|Patch|GetPatches))
  ;[0.01703 0.15 0.1262 0.5] Increase the lifespan of patches nearby refresh location. Increase based on how close to refresh location patch is
  (for _array_element (PatchHelpers|GetNearbyPatches RefreshLocation)
    (if (not (Math|Float|NearlyEqual(Float) (.SpawnTime_5_00AB2D44400CB05FDC7265903521A906 (Utilities|Struct|BreakIcyPatchData (Utilities|Array|Get(aref) _patches _array_element))) (Fortnite|Sync|Time|GetServerWorldTimeSeconds)))
      (Utilities|Array|SetArrayElem _patches _array_element (DataHelpers|SetPatchSpawnTime (Utilities|Array|Get(aref) _patches _array_element) (Math|Float|Min(Float) (Fortnite|Sync|Time|GetServerWorldTimeSeconds) (+ (.SpawnTime_5_00AB2D44400CB05FDC7265903521A906 (Utilities|Struct|BreakIcyPatchData (Utilities|Array|Get(aref) _patches _array_element))) (* (Variables|Settings|Patch|GetPatchLifeSpan) (Math|Float|Power (- 1.0 (/ (Math|Vector|Distance(Vector) (Utilities|Struct|BreakIcyPatchData (Utilities|Array|Get(aref) _patches _array_element)) (Variables|GetRefreshLocation)) (Math|Float|Sqrt (Variables|Settings|Patch|GetNearbyRadiusSquared)))) (Variables|Settings|Patch|GetRefreshCurveExponent))))))))))

(fn RequestEQSQuery
  (params (EQS%20Query :type Env%20Query%20Object%20Reference) (Query%20Location :type Vector))
  (Utilities|IsValid (Variables|Runtime|EQS|GetEQSQueryInstance)
    (:Is%20Valid
      (Utilities|Map|Add (Variables|Runtime|EQS|GetQueuedEQSQueries) EQS%20Query Query%20Location))
    (:Is%20Not%20Valid
      (Transformation|SetActorLocation Query%20Location false true) ; So EQS will use correct location
      (bind _returnvalue (AI|EQS|RunEQSQuery EQS%20Query self "AllMatching"))
      (Utilities|IsValid _returnvalue
        (:Is%20Valid
          (bind _output_get (Variables|Runtime|EQS|SetEQSQueryInstance _returnvalue))
          (AI|EQS|BindEventtoOnQueryFinished _output_get (EventDispatchers|CreateEvent :!event PatchPlacementEQSFinished)))))))

(fn RunNextEQSInQueue
  (bind _queuedeqsqueries (Variables|Runtime|EQS|GetQueuedEQSQueries))
  (bind (_key _value) (Utilities|Map|GetKeyValueByIndex _queuedeqsqueries))
  (if (not (Utilities|Map|IsEmpty _queuedeqsqueries))
    (EQS|RequestEQSQuery _key _value)
    (Utilities|Map|Remove (Variables|Runtime|EQS|GetQueuedEQSQueries) _key)))

(fn SetPatchDecal
  (params (InPatchData :type Icy%20Patch%20Data%20Structure%20%28by%20ref%29) (Decal :type Decal%20Component%20Object%20Reference))
  (bind (_location_2_4a09594d44e582f062dc1cabe0e0b05d _spawntime_5_00ab2d44400cb05fdc7265903521a906 _particleindex_8_dcfdbc164b67f81dded7538feb3ab70d) (Utilities|Struct|BreakIcyPatchData InPatchData))
  (return (Utilities|Struct|MakeIcyPatchData _location_2_4a09594d44e582f062dc1cabe0e0b05d _spawntime_5_00ab2d44400cb05fdc7265903521a906 _particleindex_8_dcfdbc164b67f81dded7538feb3ab70d Decal)))

(fn SetPatchParticle
  (params (InPatchData :type Icy%20Patch%20Data%20Structure%20%28by%20ref%29) (ParticleIndex :type Integer))
  (bind (_location_2_4a09594d44e582f062dc1cabe0e0b05d _spawntime_5_00ab2d44400cb05fdc7265903521a906 _particleindex_8_dcfdbc164b67f81dded7538feb3ab70d _decal_12_437fe3e24897c0cbaffacaba6213caab) (Utilities|Struct|BreakIcyPatchData InPatchData))
  (return (Utilities|Struct|MakeIcyPatchData _location_2_4a09594d44e582f062dc1cabe0e0b05d _spawntime_5_00ab2d44400cb05fdc7265903521a906 ParticleIndex _decal_12_437fe3e24897c0cbaffacaba6213caab)))

(fn SetPatchSpawnTime
  (params (InPatchData :type Icy%20Patch%20Data%20Structure%20%28by%20ref%29) (SpawnTime :type Float%20%28double-precision%29))
  (bind (_location_2_4a09594d44e582f062dc1cabe0e0b05d _spawntime_5_00ab2d44400cb05fdc7265903521a906 _particleindex_8_dcfdbc164b67f81dded7538feb3ab70d _decal_12_437fe3e24897c0cbaffacaba6213caab) (Utilities|Struct|BreakIcyPatchData InPatchData))
  (Utilities|FlowControl|Sequence
    (:then_0
      (Utilities|IsValid _decal_12_437fe3e24897c0cbaffacaba6213caab
        (:Is%20Valid
          (Rendering|Components|Decal|SetFadeOut _decal_12_437fe3e24897c0cbaffacaba6213caab (+ (- SpawnTime (Fortnite|Sync|Time|GetServerWorldTimeSeconds)) (Variables|Settings|Patch|GetPatchLifeSpan)) (Variables|Settings|Decal|GetDecalFadeOutTime) false))))
    (:then_1
      (return (Utilities|Struct|MakeIcyPatchData _location_2_4a09594d44e582f062dc1cabe0e0b05d SpawnTime _particleindex_8_dcfdbc164b67f81dded7538feb3ab70d _decal_12_437fe3e24897c0cbaffacaba6213caab)))))

(fn ShouldAddLocationToPatch
  (params (Location :type Vector) (CombinationRadius :type Float%20%28double-precision%29))
  (bind (_nearbypatcheindices _boverlappingpatch) (PatchHelpers|GetNearbyPatches Location (Math|Float|Square CombinationRadius) true))
  ;[0.04259 0.05316 0.15 0.5] Is location within range of a patch held by this actor
  (return _boverlappingpatch))

(fn ConstructionScript
  ; Disable collision on clients
  (Networking|SwitchHasAuthority
    (:Remote
      (Collision|SetCollisionEnabled (Variables|Default|GetBox) "NoCollision"))))
