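# EQS query that finds placement points for an icy patch in the TeaCakeIcyShins gameplay feature.
# A small circle generator spaces candidate points by distance and projects them downward onto geometry using a custom game trace channel.
# Fortnite surface and obstruction tests check for flat, valid ground and filter out points blocked from the querier.
# A raised box overlap test on that same channel both filters and scores the candidates.
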EnvQuery /TeaCakeIcyShins/Gameplay/IcyPatch/EQS_IcyPatchPlacement
options [{
  generator {
   name=EnvQueryGenerator_OnCircle_0
   class=EnvQueryGenerator_OnCircle
   properties {
    circleCenterZOffset=75
    circleRadius=60
    numberOfPoints=8
    pointOnCircleSpacingMethod=BySpaceBetween
    spaceBetween=90
    projectionData {
     extentX=0
     navigationFilter=null
     postProjectionVerticalOffset=10
     projectDown=256
     projectUp=0
     serializedChannel=ECC_GameTraceChannel11
     traceChannel=TraceTypeQuery10
     traceMode=GeometryByChannel
    }
    traceData {navigationFilter=null}
   }
  }
  tests [
   FortQueryTest_ValidSurface_0 {
    class=FortQueryTest_ValidSurface
    properties {
     flatSurfaceToleranceZ=75
     radius=100
     traceCollisionChannel=ECC_GameTraceChannel11
     traceOffsetDown=100
     traceOffsetUp=100
     surfaceMaterials []
    }
   }
   FortQueryTest_IsObstructed_0 {
    class=FortQueryTest_IsObstructed
    properties {
     context=&EnvQueryContext_Querier
     contextHeightOffset=10
     testPurpose=Filter
     traceFromContext=true
     traceData {serializedChannel=ECC_GameTraceChannel11, traceChannel=TraceTypeQuery10}
    }
   }
   EnvQueryTest_Overlap_0 {
    class=EnvQueryTest_Overlap
    properties {testPurpose=FilterAndScore, overlapData {extentX=75, extentY=75, extentZ=75, overlapChannel=ECC_GameTraceChannel11, shapeOffset=<0 0 100>}}
   }
  ]
 }]
